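; Blueprint base for the PinkOatmeal rift bubble environmental item, containing only construction-script logic that sizes its visual bounds.
; The construction script sets the world scale of the BoundsVisualBubble component to BoundsVisualBubbleScale multiplied by BoundsMeshRadius.
; This ties the displayed bubble size to the bounds radius variable.
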(fn ConstructionScript
  (Transformation|SetWorldScale3D (Variables|Default|GetBoundsVisualBubble) (* (Variables|Default|GetBoundsVisualBubbleScale) (Variables|Default|GetBoundsMeshRadius))))
