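; Base Blueprint for the PinkOatmeal environmental volume, which gives Athena player pawns a quest "inside" gameplay effect while they are in it.
; Bounds overlap events call enter and exit handlers.
; Entry applies the effect if a creative-mode check allows quest completion, and exit removes it unless the pawn is dead.
; The construction script uniformly scales the bounds by the mesh radius, and GetBubbleRadius derives a radius from it and the visual bubble scale.
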(event OnComponentBeginOverlap(BoundsComponent)
  (params (OverlappedComponent :type Primitive%20Component%20Object%20Reference) (OtherActor :type Actor%20Object%20Reference) (OtherComp :type Primitive%20Component%20Object%20Reference) (OtherBodyIndex :type Integer) (bFromSweep :type Boolean) (SweepResult :type Hit%20Result%20Structure%20%28by%20ref%29))
  (CallFunction|OnActorEnterPinkOatmeal OtherActor))

(event OnComponentEndOverlap(BoundsComponent)
  (params (OverlappedComponent :type Primitive%20Component%20Object%20Reference) (OtherActor :type Actor%20Object%20Reference) (OtherComp :type Primitive%20Component%20Object%20Reference) (OtherBodyIndex :type Integer))
  (CallFunction|OnActorExitPinkOatmeal OtherActor))

(event Custom|OnActorEnterPinkOatmeal
  (params (Actor :type Actor%20Object%20Reference) (ActorComponent :type Actor%20Component%20Object%20Reference) (InstigatorPawn :type Pawn%20Object%20Reference) (Volume :type Fort%20Volume%20Object%20Reference))
  (bind _cancompletequest (Class|AthenaQuestFunctionLibrary|AthenaCheckCreativeMode (Game|GetGameState)))
  (if _cancompletequest
    (bind _asplayer_pawn_athena (Utilities|Casting|CastToPlayerPawn_Athena Actor))
    (Environment|ApplyGlobalEnvironmentGameplayEffecttoActor _asplayer_pawn_athena "/Game/Athena/Items/EnvironmentalItems/PinkOatmeal/GE/GE_PinkOatmeal_Quest_Inside.GE_PinkOatmeal_Quest_Inside_C")))

(event Custom|OnActorExitPinkOatmeal
  (params (Actor :type Actor%20Object%20Reference) (ActorComponent :type Actor%20Component%20Object%20Reference) (InstigatorPawn :type Pawn%20Object%20Reference) (Volume :type Fort%20Volume%20Object%20Reference))
  (bind _asplayer_pawn_athena (Utilities|Casting|CastToPlayerPawn_Athena Actor))
  (if (not (FortPawn|IsDead _asplayer_pawn_athena))
    (GameplayEffects|RemoveActiveGameplayEffectbySourceEffect (Ability|GetAbilitySystemComponent _asplayer_pawn_athena) "/Game/Athena/Items/EnvironmentalItems/PinkOatmeal/GE/GE_PinkOatmeal_Quest_Inside.GE_PinkOatmeal_Quest_Inside_C" 0 -1)))

(fn GetBubbleRadius
  (bind _boundsvisualbubblescale_x (Math|Vector|BreakVector (Variables|GetBoundsVisualBubbleScale_X)))
  (return (* (Variables|Default|GetBoundsMeshRadius) (* _boundsvisualbubblescale_x 1e+03))))

(fn ConstructionScript
  (Transformation|SetWorldScale3D (Variables|Default|GetBoundsComponent) (Math|Vector|MakeVector (Variables|Default|GetBoundsMeshRadius) (Variables|Default|GetBoundsMeshRadius) (Variables|Default|GetBoundsMeshRadius))))
